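# Niagara swing effect for the ChemPencil pickaxe, pairing animation-trail ribbons with chemical bubbles.
# SwingTrail draws a short-lived grey ribbon, while TopVialTrails draws two ribbons colored pink and green by spawn group.
# Bubbles spawns teal or pink particles at a socket on a curve-driven rate, inheriting some socket velocity, then drifting upward under drag.
# Late-life curves drive splash, wave and ripple parameters for mesh and sprite renderers, with ripples tinted by time of day.
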NiagaraSystem /BRCosmetics/Weapons/FORT_Melee/Pickaxe_ChemPencil/FX/NS_FX_ChemPencilPickaxe_Swing
dependencies {
 usedDataInterfaces [&NiagaraDataInterfaceCurve, &NiagaraDataInterfaceColorCurve, &NiagaraDataInterfaceSocketReader]
 usedDynamicInputs [&Multiply_LinearColorByFloat]
 usedModules [
  &self:SystemSpawnScript.SystemScriptSource.SystemScriptGraph.NiagaraNodeAssignment_0.SetVariables_#1
  &self:SystemSpawnScript.SystemScriptSource.SystemScriptGraph.NiagaraNodeAssignment_1.SetVariables_#2
  &self:SystemSpawnScript.SystemScriptSource.SystemScriptGraph.NiagaraNodeAssignment_2.SetVariables_#3
  &SystemState
 ]
}
emitters {
 Bubbles {
  emitterUpdateScript [
   &EmitterState(Life Cycle Mode=::NewEnumerator1, Loop Behavior=::NewEnumerator1, Loop Delay=0.025, Loop Duration=0.15, UseLoopDelay=true)
   &SpawnRate(
    SpawnRate=&FloatFromCurve(
     FloatCurve=DataInterface() {propertyValues {Curve {keys [
         <at 0 1 RCIM_Cubic>
         {arriveTangent=-1, interpMode=RCIM_Cubic, leaveTangent=-1, time=0.35, value=1}
         {interpMode=RCIM_Cubic, leaveTangent=-8.0089, time=1, value=0}
        ]}}}
     CurveIndex=$Emitter.NormalizedLoopAge
     Scale Curve=50
    )
    Use Spawn Probability=true
   )
  ]
  particleSpawnScript [
   /SetVariables(Particles.SpawnGroup=&RandomRangeInt(Minimum=0, Maximum=1)) {isSetParameters=true}
   &InitializeParticle(
    Color=&MergeHelper_ColorSwitch(0=<rgba 0 2 0.2165 1>, 1=<rgba 2 0 1.2763 1>, 2=?, 3=?, Index=$Particles.SpawnGroup, Num=::NewEnumerator0)
    Lifetime Max=0.8
    Lifetime Min=0.3
    Lifetime Mode=::NewEnumerator1
    Mesh Scale Mode=::NewEnumerator1
    Mesh Uniform Scale Max=0.1
    Mesh Uniform Scale Min=0.06
    Sprite Rotation Mode=::NewEnumerator2
    UsePositionOffset=false
    Write Material Random=true
   )
   &SocketLocation(
    Interpolate Socket Sampling=true
    SamplePosOffset=false
    Socket Index=4
    Socket Reader=$System.Socket Reader
    Socket Sampling Mode=::NewEnumerator3
    VelocitySpeedLimit=false
   )
   &ShapeLocation(MaskSpawnGroup=false, Override Seed=false, Sphere Radius=15, UseOwnerScale=true)
   /SetVariables() {isSetParameters=true}
   &InheritVelocity(
    Inherited Velocity Amount Scale=&VectorFromFloat(Value=&RandomRangeFloat(Minimum=0, Maximum=0.1))
    Inherited Velocity Speed Limit=222
    Speed Limit=true
    UseSpeedThreshold=false
    Velocity Source=$Output.SocketLocation.SampledVelocity
   )
   &AddVelocity(
    Module.Use Falloff Distance=false
    Override Seed=false
    Velocity Mode=::NewEnumerator1
    Velocity Origin=$StackContext.ShapeLocation.ShapePosition
    Velocity Speed=&RandomRangeFloat(Minimum=5, Maximum=12.5)
   )
  ]
  particleUpdateScript [
   &ParticleState()
   &ScaleMeshSize(Scale Factor=&VectorFromFloat(Value=&FloatFromCurve(FloatCurve=DataInterface() {propertyValues {Curve {keys [<at 0 0.2>,<at 0.8 1>,<at 1 0>]}}}, Scale Curve=0.35)))
   &Drag(Drag=2.5, Use Linear Drag=true, Use Rotational Drag=true)
   &DynamicMaterialParameters(
    Index 0 Param 1=&FloatFromCurve(FloatCurve=DataInterface() {propertyValues {Curve {keys [
         <at 0 0>
         {interpMode=RCIM_Cubic, leaveTangent=1.2072, tangentMode=RCTM_Break, time=0.55, value=0}
         {arriveTangent=3.283, tangentMode=RCTM_User, time=1, value=1}
        ]}}})
    Index 0 Param 4=0.2
    Module.Index0Param0Write=true
    Module.Index0Param1Write=true
    Module.Index0Param2Write=true
    Module.Index0Param3Write=true
   )
   &AccelerationForce(Acceleration=&Multiply_VectorByFloat(
     Vector=&RandomRangeVector(Minimum=<0 0 0>, Maximum=<0 0 150>)
     Float=&FloatFromCurve(FloatCurve=DataInterface() {propertyValues {Curve {keys [
          {arriveTangent=4.1785, interpMode=RCIM_Cubic, leaveTangent=4.1785, tangentMode=RCTM_User, time=0, value=0}
          {arriveTangent=1.1179, interpMode=RCIM_Cubic, leaveTangent=1.1179, tangentMode=RCTM_User, time=0.2042, value=0.6228}
          {arriveTangent=0.0224, interpMode=RCIM_Cubic, leaveTangent=0.0224, tangentMode=RCTM_User, time=1, value=1}
         ]}}})
    ))
   &SolveForcesAndVelocity(Clamp Velocity=false, Limit Acceleration=false)
   /SetVariables(
    Particles.SplashDynamicParamenter:Vector4f=&MakeVector4(
     X=&FloatFromCurve(FloatCurve=DataInterface() {propertyValues {Curve {keys [
          <at 0 1>
          {interpMode=RCIM_Cubic, leaveTangent=34.4816, tangentMode=RCTM_Break, time=0.8, value=1}
          {arriveTangent=125.3655, tangentMode=RCTM_User, time=1, value=17}
         ]}}})
     Y=0.7
     Z=0
     W=0
    )
    Particles.SplashIndex:NiagaraFloat=&RandomRangeFloat(Minimum=0, Maximum=4)
    Particles.SplashSize:Vector2f=&Vector2DFromFloat(Value=&FloatFromCurve(
      FloatCurve=DataInterface() {propertyValues {Curve {keys [<at 0 0>,<at 0.7 0>,<at 0.8 1>]}}}
      Scale Curve=&RandomRangeFloat(Minimum=15, Maximum=17)
     ))
   ) {isSetParameters=true}
   /SetVariables(
    Particles.WaveColor:LinearColor=&ColorFromCurve(ColorCurve=DataInterface() {propertyValues {
       AlphaCurve {keys [<at 0.6765 0>,<at 0.7462 1>,<at 0.9338 0 RCIM_Cubic>]}
       BlueCurve {keys [<at 0.8093 0.589 RCIM_Cubic>,<at 1 0.1407 RCIM_Cubic>]}
       GreenCurve {keys [<at 0.8093 1 RCIM_Cubic>,<at 1 0.3542 RCIM_Cubic>]}
       RedCurve {keys [<at 0.8093 0.7346 RCIM_Cubic>,<at 1 0.1829 RCIM_Cubic>]}
      }})
    Particles.WaveSize:Vector2f=&Vector2DFromFloat(Value=&MakeFloatFromVector2d(Vector 2D=&Vector2DFromFloat(Value=&FloatFromCurve(
        FloatCurve=DataInterface() {propertyValues {Curve {keys [<at 0 0>,<at 0.7 0>,<at 0.8 1>]}}}
        Scale Curve=&RandomRangeFloat(Minimum=15, Maximum=17)
       ))))
   ) {isSetParameters=true}
   /SetVariables(
    Particles.RippleColor:LinearColor=&MultiplyLinearColor(
     Linear Color=&ColorFromCurve(ColorCurve=DataInterface() {propertyValues {
        AlphaCurve {keys [<at 0.6765 0>,<at 0.7462 0.3>,<at 0.9056 0 RCIM_Cubic>]}
        BlueCurve {keys [<at 0.8093 0.2301 RCIM_Cubic>,<at 1 0.1143 RCIM_Cubic>]}
        GreenCurve {keys [<at 0.8093 0.7552 RCIM_Cubic>,<at 1 0.3542 RCIM_Cubic>]}
        RedCurve {keys [<at 0.8093 0.4172 RCIM_Cubic>,<at 1 0.2272 RCIM_Cubic>]}
       }})
     Scale Factor=&MakeVector4_FromLinearColor(LinearColor=&FNTimeOfDayColor(Night Color=<rgba 0.556 0.5008 1 1>, Morning Color=<rgba 1 0.6695 0.6095 1>, Day Color=<rgba 1 1 1 1>, Evening Color=<rgba 1 1 1 1>))
    )
    Particles.RippleFacing:Vector3f=<0 0 1>
    Particles.RippleSize:Vector2f=&Vector2DFromFloat(Value=&FloatFromCurve(FloatCurve=DataInterface() {propertyValues {Curve {keys [<at 0.5 0>,<at 1 1>]}}}, Scale Curve=35))
   ) {isSetParameters=true}
   /SetVariables(
    Particles.Ripple2Color:LinearColor=&MultiplyLinearColor(
     Linear Color=&ColorFromCurve(ColorCurve=DataInterface() {propertyValues {
        AlphaCurve {keys [<at 0.8143 0>,<at 0.8876 0.2>,<at 0.9661 0 RCIM_Cubic>]}
        BlueCurve {keys [<at 0.845 0.1745 RCIM_Cubic>,<at 1 0.0925 RCIM_Cubic>]}
        GreenCurve {keys [<at 0.845 0.7396 RCIM_Cubic>,<at 1 0.2656 RCIM_Cubic>]}
        RedCurve {keys [<at 0.845 0.3534 RCIM_Cubic>,<at 1 0.1417 RCIM_Cubic>]}
       }})
     Scale Factor=&MakeVector4_FromLinearColor(LinearColor=&FNTimeOfDayColor(Night Color=<rgba 0.556 0.5008 1 1>, Morning Color=<rgba 1 0.6695 0.6095 1>, Day Color=<rgba 1 1 1 1>, Evening Color=<rgba 1 1 1 1>))
    )
    Particles.Ripple2Size:Vector2f=&Vector2DFromFloat(Value=&FloatFromCurve(
      FloatCurve=DataInterface() {propertyValues {Curve {keys [
          {interpMode=RCIM_Cubic, leaveTangent=0.6678, tangentMode=RCTM_Break, time=0.6, value=0}
          {arriveTangent=2.5069, tangentMode=RCTM_User, time=1, value=0.65}
         ]}}}
      Scale Curve=30
     ))
   ) {isSetParameters=true}
  ]
  rendererClasses [&NiagaraMeshRendererProperties, &NiagaraSpriteRendererProperties]
  renderers [
   {rendererClass=&NiagaraMeshRendererProperties, rendererIndex=0}
   {rendererClass=&NiagaraSpriteRendererProperties, rendererIndex=1}
   {rendererClass=&NiagaraSpriteRendererProperties, rendererIndex=2}
   {rendererClass=&NiagaraMeshRendererProperties, rendererIndex=3}
  ]
 }
 SwingTrail {
  emitterSpawnScript [&FN_RibbonIDOffsetUtility()]
  emitterUpdateScript [
   &EmitterState()
   &FN_SpawnAnimTrail()
  ]
  particleSpawnScript [
   &FN_InitializeAnimTrailBoneIndecies()
   &FN_InitializeAnimTrailBoneIndecies(AnimTrail Selection=1, First Bone or Socket Index=2, Second Bone or Socket Index=2) {enabled=false, module=FN_InitializeAnimTrailBoneIndecies001}
   &FN_InitializeAnimTrailBoneIndecies(AnimTrail Selection=1, First Bone or Socket Index=3, Second Bone or Socket Index=3) {enabled=false, module=FN_InitializeAnimTrailBoneIndecies002}
   &FN_InitializeAnimTrail(AnimTrailWidth=1.5, Color=<rgba 0.2969 0.2969 0.2969 1>, Lifetime=0.25)
   &AddVelocity(Velocity=<0 400 0>) {enabled=false}
  ]
  particleUpdateScript [
   &SolveForcesAndVelocity(Clamp Velocity=false, Limit Acceleration=false)
   &FN_UpdateAnimTrail() {module=FN_UpdateAnimTrail001}
   &ScaleColor(
    Scale Alpha=&Multiply_Float(
     A=&FloatFromCurve(
      FloatCurve=DataInterface() {propertyValues {Curve {keys [
          {interpMode=RCIM_Cubic, leaveTangent=-2.3646, tangentMode=RCTM_Break, time=0, value=1}
          {arriveTangent=-0.2449, tangentMode=RCTM_User, time=0.7, value=0.05}
          {arriveTangent=-0.2139, interpMode=RCIM_Cubic, tangentMode=RCTM_Break, time=1, value=0}
         ]}}}
      Scale Curve=2
     )
     B=1
    )
    ScaleA=true
    ScaleRGB=true
   )
  ]
  rendererClasses [&NiagaraRibbonRendererProperties]
  renderers [
   {rendererClass=&NiagaraRibbonRendererProperties}
   {rendererClass=&NiagaraRibbonRendererProperties, rendererIndex=1}
   {rendererClass=&NiagaraRibbonRendererProperties, rendererIndex=2}
  ]
 }
 TopVialTrails {
  emitterSpawnScript [&FN_RibbonIDOffsetUtility()]
  emitterUpdateScript [
   &EmitterState()
   &FN_SpawnAnimTrail(Number of Anim Trails=2)
  ]
  particleSpawnScript [
   &FN_InitializeAnimTrailBoneIndecies(First Bone or Socket Index=2, Second Bone or Socket Index=2)
   &FN_InitializeAnimTrailBoneIndecies(AnimTrail Selection=1, First Bone or Socket Index=3, Second Bone or Socket Index=3) {module=FN_InitializeAnimTrailBoneIndecies001}
   &FN_InitializeAnimTrailBoneIndecies(AnimTrail Selection=1, First Bone or Socket Index=3, Second Bone or Socket Index=3) {enabled=false, module=FN_InitializeAnimTrailBoneIndecies002}
   &FN_InitializeAnimTrail(Lifetime=0.3)
   &AddVelocity(Velocity=<0 400 0>) {enabled=false}
  ]
  particleUpdateScript [
   &SolveForcesAndVelocity(Clamp Velocity=false, Limit Acceleration=false)
   &FN_UpdateAnimTrail() {module=FN_UpdateAnimTrail001}
   &Color(Color=&MergeHelper_ColorSwitch(0=<rgba 1 0 0.2633 1>, 1=<rgba 0 1 0.1215 1>, 2=?, 3=?, Index=$Particles.SpawnGroup, Num=::NewEnumerator0))
   &ScaleColor(
    Color Value To Scale=$Particles.Color
    Scale Alpha=&Multiply_Float(
     A=&FloatFromCurve(FloatCurve=DataInterface() {propertyValues {Curve {keys [
          {interpMode=RCIM_Cubic, leaveTangent=-2.3646, tangentMode=RCTM_Break, time=0, value=1}
          {arriveTangent=-0.2449, tangentMode=RCTM_User, time=0.7, value=0.05}
          {arriveTangent=-0.2139, interpMode=RCIM_Cubic, tangentMode=RCTM_Break, time=1, value=0}
         ]}}})
     B=1
    )
    Scale RGB=&VectorFromFloat(Value=&FloatFromCurve(FloatCurve=DataInterface() {propertyValues {Curve {keys [<at 0 5>,<at 1 1>]}}}))
    ScaleA=true
    ScaleRGB=true
   )
   /SetVariables(Particles.RibbonWidth=11) {isSetParameters=true}
  ]
  rendererClasses [&NiagaraRibbonRendererProperties]
  renderers [
   {rendererClass=&NiagaraRibbonRendererProperties}
   {rendererClass=&NiagaraRibbonRendererProperties, rendererIndex=1}
   {rendererClass=&NiagaraRibbonRendererProperties, rendererIndex=2}
  ]
 }
}
summary {
 systemName=NS_FX_ChemPencilPickaxe_Swing
 emitters [
  {emitterName=SwingTrail, rendererClasses [&NiagaraRibbonRendererProperties]}
  {emitterName=TopVialTrails, rendererClasses [&NiagaraRibbonRendererProperties]}
  {emitterName=Bubbles, rendererClasses [&NiagaraMeshRendererProperties, &NiagaraSpriteRendererProperties]}
 ]
 userVariables [
  User.bIsFrontEnd {type=NiagaraBool}
  User.bIsFrontEndPreview {type=NiagaraBool}
  User.RibbonIDOffset {type=NiagaraInt32}
 ]
}
